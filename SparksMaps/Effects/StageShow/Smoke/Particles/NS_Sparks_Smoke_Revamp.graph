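# Niagara smoke effect for the Sparks stage show, shooting a colored vertical plume upward and tinting all layers from a shared system color.
# A Base mesh and a GPU flipbook sprite Fill form the body, beneath short-fast and tall-slow ribbon streams slowed by rising drag.
# Ribbons spawn at frame-time-compensated rates, order links by counter, and drive panning, eroding material parameters with per-instance random offsets.
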NiagaraSystem /SparksMaps/Effects/StageShow/Smoke/Particles/NS_Sparks_Smoke_Revamp
dependencies {
 usedDataInterfaces [
  &NiagaraDataInterfaceCurve
  &NiagaraDataInterfaceArrayFloat2
  &NiagaraDataInterfaceVector2DCurve
  &NiagaraDataInterfaceVectorCurve
  /Script/Niagara.NiagaraDataInterfaceSimpleCounter
 ]
 usedDynamicInputs [
  &SimulationPosition
  &UniformRangedFloat
  &self:NormalizeColor
  &HueShiftLinearColor
  &RandomRangeLinearColor
 ]
 usedModules [
  &self:SystemSpawnScript.SystemScriptSource.SystemScriptGraph.NiagaraNodeAssignment_2.SetVariables_#1
  &self:SystemSpawnScript.SystemScriptSource.SystemScriptGraph.NiagaraNodeAssignment_5.SetVariables_#2
  &self:FindHue
  &SystemState
 ]
}
emitters {
 Base {
  emitterUpdateScript [
   &EmitterState()
   &SpawnBurst_Instantaneous(Use Spawn Probability=false, UseLoopCountLimit=false)
   /SetVariables(Emitter.Opacity:NiagaraFloat=&FloatFromCurve(
     FloatCurve=DataInterface() {propertyValues {Curve {keys [<at 1.25 0.2 RCIM_Cubic>, {arriveTangent=16.0497, tangentMode=RCTM_User, time=1.6, value=3}]}}}
     CurveIndex=$Emitter.LoopedAge
    )) {isSetParameters=true}
  ]
  particleSpawnScript [
   &InitializeParticle(
    Color=&MultiplyLinearColor(
     Linear Color=<rgba 0.449 0.449 0.449 1>
     Scale Factor=&MakeVector4_FromLinearColor(LinearColor=$System.SystemColor)
    )
    Lifetime=1.75
    Mesh Scale=<1.9 0.5 0.5>
    Mesh Scale Mode=::NewEnumerator4
    Position Offset=<0 0 -30>
    UsePositionOffset=true
    Write Material Random=true
   )
   &InitialMeshOrientation(Mesh Orientation Mode=::NewEnumerator2, Orientation Vector=<0 0 1>, Use Rotation Vector=false)
  ]
  particleUpdateScript [
   &ParticleState()
   &ScaleMeshSize(Scale Factor=&VectorFromFloat(Value=&FloatFromCurve(FloatCurve=DataInterface() {propertyValues {Curve {keys [{arriveTangent=7.1495, interpMode=RCIM_Cubic, leaveTangent=7.1495, tangentMode=RCTM_User, time=0, value=0.2}, <at 0.1 1>]}}})))
   &ScaleColor(
    Scale Alpha=&FloatFromCurve(FloatCurve=DataInterface() {propertyValues {Curve {keys [
         {tangentMode=RCTM_User, time=0, value=1}
         {arriveTangent=-0.1905, interpMode=RCIM_Cubic, leaveTangent=-0.1905, tangentMode=RCTM_User, time=0.82, value=1}
         {arriveTangent=-4.571, interpMode=RCIM_Cubic, leaveTangent=-4.571, tangentMode=RCTM_User, time=1, value=0}
        ]}}})
    ScaleA=true
    ScaleRGB=false
   )
   &DynamicMaterialParameters(
    Index 0 Param 2=-3
    Module.Index0Param0Write=true
    Module.Index0Param1Write=true
    Module.Index0Param2Write=true
    Module.Index0Param3Write=true
   )
  ]
  rendererClasses [&NiagaraMeshRendererProperties]
  renderers [{rendererClass=&NiagaraMeshRendererProperties, rendererIndex=0}]
 }
 Fill {
  simTarget=GPUComputeSim
  emitterSpawnScript [&self:CrunchColor(Color Adjust=0.1) {module=CrunchColor}]
  emitterUpdateScript [
   &EmitterState()
   &SpawnBurst_Instantaneous(Spawn Time=0.2, Use Spawn Probability=false, UseLoopCountLimit=false)
  ]
  particleSpawnScript [
   &InitializeParticle(
    Color=&MultiplyLinearColor(
     Linear Color=<rgba 0.6806 0.6867 0.8198 1>
     Scale Factor=&MakeVector4_FromLinearColor(LinearColor=$System.SystemColor)
    )
    Lifetime=1.9
    Position Offset=<0 0 -30>
    Sprite Size=&RandomRangeVector2D(Minimum=<600 750>, Maximum=<700 850>)
    Sprite Size Mode=::NewEnumerator4
    Sprite UV Mode=::NewEnumerator1
    UsePositionOffset=true
    Write Material Random=true
   ) {module=InitializeParticle001}
   /SetVariables(Particles.UVScale:Vector2f=&SelectVector2DFromArray(
     Array Sampling Mode=::NewEnumerator0
     Direct Array Index=?
     Vector 2D Selection Array=DataInterface() {propertyValues {FloatData [<-1 1>, <1 1>]}}
     Random Seed=0
    )) {isSetParameters=true}
   &AddVelocity(Velocity=<0 0 1>)
  ]
  particleUpdateScript [
   &ParticleState()
   &SolveForcesAndVelocity(Clamp Velocity=false, Limit Acceleration=false)
   &ScaleColor(
    Scale Alpha=&FloatFromCurve(FloatCurve=DataInterface() {propertyValues {Curve {keys [
         {arriveTangent=12.8921, interpMode=RCIM_Cubic, leaveTangent=12.8921, tangentMode=RCTM_User, time=0, value=0}
         {arriveTangent=3.1086, interpMode=RCIM_Cubic, leaveTangent=3.1086, tangentMode=RCTM_User, time=0.085, value=0.7}
         {arriveTangent=-0.4531, interpMode=RCIM_Cubic, leaveTangent=-0.4531, tangentMode=RCTM_User, time=0.5, value=0.8}
         {arriveTangent=-3.3296, interpMode=RCIM_Cubic, leaveTangent=-3.3296, tangentMode=RCTM_User, time=1, value=0}
        ]}}})
    ScaleA=true
    ScaleRGB=false
   )
   &ScaleSpriteSize(
    Non-Uniform Curve Sprite Scale=DataInterface() {propertyValues {
      XCurve {keys [
        {arriveTangent=2.0712, interpMode=RCIM_Cubic, leaveTangent=2.0712, tangentMode=RCTM_User, time=0, value=0.3}
        {arriveTangent=0.7873, interpMode=RCIM_Cubic, leaveTangent=0.7873, tangentMode=RCTM_User, time=0.35, value=0.8}
        {arriveTangent=0.1738, tangentMode=RCTM_User, time=1, value=1.05}
       ]}
      YCurve {keys [
        {arriveTangent=7.4932, interpMode=RCIM_Cubic, leaveTangent=7.4932, tangentMode=RCTM_User, time=0, value=0}
        {arriveTangent=0.9767, interpMode=RCIM_Cubic, leaveTangent=0.9767, tangentMode=RCTM_User, time=0.2, value=0.9}
        {arriveTangent=0.0542, tangentMode=RCTM_User, time=1, value=1.25}
       ]}
     }}
    Scale Sprite Size Mode=::NewEnumerator3
   ) {module=ScaleSpriteSize001}
   /SetVariables(Particles.SubImageIndex=&Add_Float(
     A=&FloatFromCurve(
      FloatCurve=DataInterface() {propertyValues {Curve {keys [
          {arriveTangent=1.513, interpMode=RCIM_Cubic, leaveTangent=1.513, tangentMode=RCTM_User, time=0, value=0.07}
          {arriveTangent=0.7372, tangentMode=RCTM_User, time=1, value=1}
         ]}}}
      Scale Curve=&RandomRangeFloat(Minimum=30, Maximum=44)
     )
     B=&RandomRangeFloat(Minimum=0, Maximum=20)
    )) {isSetParameters=true}
   &DynamicMaterialParameters(
    Index 0 Param 1=&FloatFromCurve(FloatCurve=DataInterface() {propertyValues {Curve {keys [
         {arriveTangent=0.0731, interpMode=RCIM_Cubic, leaveTangent=0.4079, tangentMode=RCTM_Break, time=0, value=1.6}
         {arriveTangent=4.0982, tangentMode=RCTM_User, time=1, value=3}
        ]}}})
    Module.Index0Param0Write=true
    Module.Index0Param1Write=false
    Module.Index0Param2Write=false
    Module.Index0Param3Write=false
   )
   &SpriteFacingAndAlignment(
    Sprite Alignment=&Multiply_VectorByFloat(Vector=$Particles.Velocity, Float=-1)
    WriteAlignment=true
    WriteFacing=false
   )
   /SetVariables(Particles.Position=&VectorFromCurve(VectorCurve=DataInterface() {propertyValues {
       XCurve {keys [<at 0 0>]}
       YCurve {keys [<at 0 0>]}
       ZCurve {keys [{arriveTangent=102.6051, interpMode=RCIM_Cubic, leaveTangent=102.6051, tangentMode=RCTM_User, time=0, value=-20}, <at 1 10>]}
      }})) {isSetParameters=true}
  ]
  rendererClasses [&NiagaraSpriteRendererProperties]
  renderers [{rendererClass=&NiagaraSpriteRendererProperties}]
 }
 RibbonShortFast {
  emitterUpdateScript [
   &EmitterState()
   &SpawnBurst_Instantaneous(Use Spawn Probability=false, UseLoopCountLimit=true)
   &SpawnRate(
    SpawnRate=&FloatFromCurve(
     FloatCurve=DataInterface() {propertyValues {Curve {keys [<at 1 1 RCIM_Constant>, {arriveTangent=-19.726, tangentMode=RCTM_User, time=1.8, value=0}]}}}
     CurveIndex=$Emitter.LoopedAge
     Scale Curve=&Divide_Float(A=0.5, B=$Engine.DeltaTime)
    )
    Use Spawn Probability=false
   )
   /SetVariables(Emitter.bHasSpawned:NiagaraBool=$Output.SpawnRate.HasSpawnedThisFrame) {isSetParameters=true}
  ]
  particleSpawnScript [
   /SetVariables(Particles.RibbonLinkOrder:NiagaraFloat=&MakeFloatFromInt(Int=/Niagara/DynamicInputs/DataInterface/GenerateIntFromCounter(SimpleCounter:NiagaraDataInterfaceSimpleCounter=DataInterface() {propertyValues {GpuSyncMode=None, InitialValue=0}}))) {isSetParameters=true, module=SetVariables_#3}
   &InitializeParticle(
    Color=&MultiplyLinearColor(
     Linear Color=$System.SystemColor
     Scale Factor=&MakeFloat4FromVectorAndFloat(
      "Vector (RGB)"=<2 2 2>
      "Float (Alpha)"=&FloatFromCurve(
       FloatCurve=DataInterface() {propertyValues {Curve {keys [
           {arriveTangent=5.0728, interpMode=RCIM_Cubic, leaveTangent=5.0728, tangentMode=RCTM_User, time=0, value=0}
           {arriveTangentWeight=0.0943, interpMode=RCIM_Cubic, leaveTangentWeight=0.3771, tangentMode=RCTM_User, time=0.2, value=1}
           {interpMode=RCIM_Cubic, tangentMode=RCTM_User, time=1, value=1}
           {arriveTangent=-1.914, tangentMode=RCTM_User, time=1.7, value=0}
          ]}}}
       CurveIndex=$Emitter.LoopedAge
       Scale Curve=&Lerp_Float(
        A=0
        B=2
        Alpha=&ClampFloat(Float=&Multiply_Float(A=$Particles.RibbonLinkOrder, B=0.25))
       )
      )
     )
    )
    Lifetime=0.75
    Position Offset=<0 0 -30>
    Ribbon Width=&FloatFromCurve(
     FloatCurve=DataInterface() {propertyValues {Curve {keys [{arriveTangent=7.7059, interpMode=RCIM_Cubic, leaveTangent=7.7059, tangentMode=RCTM_User, time=0, value=0.25}, <at 0.25 1>]}}}
     CurveIndex=$Emitter.LoopedAge
     Scale Curve=300
    )
    Ribbon Width Mode=::NewEnumerator1
    UsePositionOffset=true
    Write Material Random=true
   ) {module=InitializeParticle001}
   &AddVelocity(Velocity=<0 0 1200>)
  ]
  particleUpdateScript [
   &ParticleState()
   &Drag(Drag=&Lerp_Float(A=0.5, B=5, Alpha=$Particles.NormalizedAge), Use Linear Drag=true, Use Rotational Drag=true)
   &SolveForcesAndVelocity(Clamp Velocity=false, Limit Acceleration=false)
   &ScaleSpriteSize(Scale Factor=&Vector2DFromFloat(Value=&FloatFromCurve(FloatCurve=DataInterface() {propertyValues {Curve {keys [
          {arriveTangent=-0.0047, interpMode=RCIM_Cubic, leaveTangent=-0.0047, tangentMode=RCTM_User, time=0, value=0}
          {arriveTangent=1.5585, interpMode=RCIM_Cubic, leaveTangent=1.5585, tangentMode=RCTM_User, time=0.3411, value=0.8347}
          <at 1 1 RCIM_Cubic>
         ]}}})))
   &ScaleRibbonWidth(Ribbon Width Scale=&FloatFromCurve(FloatCurve=DataInterface() {propertyValues {Curve {keys [
         {arriveTangent=2.4416, interpMode=RCIM_Cubic, leaveTangent=2.4416, tangentMode=RCTM_User, time=0, value=0}
         {interpMode=RCIM_Cubic, tangentMode=RCTM_User, time=0.6, value=1}
        ]}}}))
   &FN_RibbonMaterial(
    Color=$Particles.Initial.Color
    Color Pan U Offset:NiagaraFloat=$System.InstanceRandom
    Color Pan U Speed=-1.5
    Color Scale U=0.34
    Erosion Amount:NiagaraFloat=&Lerp_Float(A=0, B=1.5, Alpha=$Particles.NormalizedAge)
    Erosion Pan Offset U:NiagaraFloat=$System.InstanceRandom
    Erosion Pan Speed U=-1
    Erossion Scale U=0.38
    Opacity Pan U Offset:NiagaraFloat=$System.InstanceRandom
    Opacity Pan U Speed=-0.4
    Opacity Scale U=0.43
    Scale Alpha:NiagaraFloat=&FloatFromCurve(FloatCurve=DataInterface() {propertyValues {Curve {keys [
         {arriveTangent=16.7044, interpMode=RCIM_Cubic, leaveTangent=16.7044, tangentMode=RCTM_User, time=0.03, value=0}
         {arriveTangent=-2.0239, interpMode=RCIM_Cubic, leaveTangent=-2.0239, tangentMode=RCTM_User, time=0.1, value=1}
         {arriveTangent=-1.1111, interpMode=RCIM_Cubic, leaveTangent=-1.1111, time=0.8, value=0.2}
         {arriveTangent=-0.3913, interpMode=RCIM_Cubic, leaveTangent=-0.3913, tangentMode=RCTM_User, time=1, value=0}
        ]}}})
    Scale Color:Vector3f=&VectorFromFloat(Value=&FloatFromCurve(FloatCurve=DataInterface() {propertyValues {Curve {keys [{tangentMode=RCTM_User, time=0, value=1}, {tangentMode=RCTM_User, time=1, value=0}]}}}))
    TIme:NiagaraFloat=$Engine.Time
   )
   &CameraOffset(Camera Offset Amount=3)
  ]
  rendererClasses [&NiagaraRibbonRendererProperties]
  renderers [{rendererClass=&NiagaraRibbonRendererProperties}]
 }
 RibbonTallSlow {
  emitterUpdateScript [
   &EmitterState()
   &SpawnBurst_Instantaneous(Use Spawn Probability=false, UseLoopCountLimit=true)
   &SpawnRate(
    SpawnRate=&FloatFromCurve(
     FloatCurve=DataInterface() {propertyValues {Curve {keys [<at 1 1 RCIM_Constant>, {arriveTangent=-19.726, tangentMode=RCTM_User, time=1.8, value=0}]}}}
     CurveIndex=$Emitter.LoopedAge
     Scale Curve=&Divide_Float(A=0.25, B=$Engine.DeltaTime)
    )
    Use Spawn Probability=false
   )
   /SetVariables(Emitter.bHasSpawned:NiagaraBool=$Output.SpawnRate.HasSpawnedThisFrame) {isSetParameters=true}
  ]
  particleSpawnScript [
   /SetVariables(Particles.RibbonLinkOrder:NiagaraFloat=&MakeFloatFromInt(Int=&GenerateIntFromCounter(SimpleCounter:NiagaraDataInterfaceSimpleCounter=DataInterface() {propertyValues {GpuSyncMode=None, InitialValue=0}}))) {isSetParameters=true, module=SetVariables_#3}
   &InitializeParticle(
    Color=&MultiplyLinearColor(
     Linear Color=$System.SystemColor
     Scale Factor=&MakeFloat4FromVectorAndFloat(
      "Vector (RGB)"=<1 1 1>
      "Float (Alpha)"=&FloatFromCurve(
       FloatCurve=DataInterface() {propertyValues {Curve {keys [
           {arriveTangent=5.0728, interpMode=RCIM_Cubic, leaveTangent=5.0728, tangentMode=RCTM_User, time=0, value=0}
           {arriveTangentWeight=0.0943, interpMode=RCIM_Cubic, leaveTangentWeight=0.3771, tangentMode=RCTM_User, time=0.3, value=1}
           {interpMode=RCIM_Cubic, tangentMode=RCTM_User, time=1, value=1}
           {arriveTangent=-1.914, tangentMode=RCTM_User, time=1.7, value=0}
          ]}}}
       CurveIndex=$Emitter.LoopedAge
       Scale Curve=&Lerp_Float(
        A=0
        B=2
        Alpha=&ClampFloat(Float=&Multiply_Float(A=$Particles.RibbonLinkOrder, B=0.25))
       )
      )
     )
    )
    Lifetime=1.5
    Position Offset=<0 0 -30>
    Ribbon Width=&FloatFromCurve(
     FloatCurve=DataInterface() {propertyValues {Curve {keys [{arriveTangent=3.9376, interpMode=RCIM_Cubic, leaveTangent=3.9376, tangentMode=RCTM_User, time=0, value=0.5}, <at 0.25 1>]}}}
     CurveIndex=$Emitter.LoopedAge
     Scale Curve=400
    )
    Ribbon Width Mode=::NewEnumerator1
    UsePositionOffset=true
    Write Material Random=true
   ) {module=InitializeParticle001}
   &AddVelocity(Velocity=<0 0 1250>)
   &self:CrunchColor(Color Adjust=0.1, Value Mult=1) {module=CrunchColor}
  ]
  particleUpdateScript [
   &ParticleState()
   &Drag(Drag=&Lerp_Float(A=0.5, B=5, Alpha=$Particles.NormalizedAge), Use Linear Drag=true, Use Rotational Drag=true)
   &SolveForcesAndVelocity(Clamp Velocity=false, Limit Acceleration=false)
   &ScaleSpriteSize(Scale Factor=&Vector2DFromFloat(Value=&FloatFromCurve(FloatCurve=DataInterface() {propertyValues {Curve {keys [
          {arriveTangent=-0.0047, interpMode=RCIM_Cubic, leaveTangent=-0.0047, tangentMode=RCTM_User, time=0, value=0}
          {arriveTangent=1.5585, interpMode=RCIM_Cubic, leaveTangent=1.5585, tangentMode=RCTM_User, time=0.3411, value=0.8347}
          <at 1 1 RCIM_Cubic>
         ]}}})))
   &ScaleRibbonWidth(Ribbon Width Scale=&FloatFromCurve(FloatCurve=DataInterface() {propertyValues {Curve {keys [
         {arriveTangent=2.4416, interpMode=RCIM_Cubic, leaveTangent=2.4416, tangentMode=RCTM_User, time=0, value=0}
         {interpMode=RCIM_Cubic, tangentMode=RCTM_User, time=0.6, value=1}
        ]}}}))
   &FN_RibbonMaterial(
    Color=$Particles.Initial.Color
    Color Pan U Offset:NiagaraFloat=$System.InstanceRandom
    Color Pan U Speed=-0.8
    Color Scale U=0.3
    Erosion Amount:NiagaraFloat=&Lerp_Float(A=0, B=1.5, Alpha=$Particles.NormalizedAge)
    Erosion Pan Offset U:NiagaraFloat=$System.InstanceRandom
    Erosion Pan Speed U=-0.7
    Erossion Scale U=0.38
    Opacity Pan U Offset:NiagaraFloat=$System.InstanceRandom
    Opacity Pan U Speed=-0.2
    Opacity Scale U=0.43
    Scale Alpha:NiagaraFloat=&FloatFromCurve(FloatCurve=DataInterface() {propertyValues {Curve {keys [
         {arriveTangent=10.3148, interpMode=RCIM_Cubic, leaveTangent=10.3148, tangentMode=RCTM_User, time=0.01, value=0}
         {arriveTangent=-2.9786, interpMode=RCIM_Cubic, leaveTangent=-2.9786, tangentMode=RCTM_User, time=0.1, value=1}
         {arriveTangent=-0.9367, interpMode=RCIM_Cubic, leaveTangent=-0.9367, tangentMode=RCTM_User, time=0.3153, value=0.3301}
         {arriveTangent=-0.2324, interpMode=RCIM_Cubic, leaveTangent=-0.2324, tangentMode=RCTM_User, time=0.8617, value=0.12}
         {arriveTangent=-0.3913, interpMode=RCIM_Cubic, leaveTangent=-0.3913, tangentMode=RCTM_User, time=1, value=0}
        ]}}})
    Scale Color:Vector3f=&MakeVectorFromLinearColorRGB(Color=&Multiply_LinearColorByFloat(
      Linear Color=&LerpLinearColors(StartColor=<rgba 1 1 1 1>, EndColor=$Particles.CrunchColor, LerpFactor=$Particles.NormalizedAge)
      Float=&Lerp_Float(A=1, B=0.75, Alpha=$Particles.NormalizedAge)
     ))
    TIme:NiagaraFloat=$Engine.Time
   )
   &CameraOffset(Camera Offset Amount=3)
  ]
  rendererClasses [&NiagaraRibbonRendererProperties]
  renderers [{rendererClass=&NiagaraRibbonRendererProperties}]
 }
}
summary {
 systemName=NS_Sparks_Smoke_Revamp
 emitters [
  {emitterName=Base, rendererClasses [&NiagaraMeshRendererProperties]}
  {emitterName=Fill, simTarget=GPUComputeSim, rendererClasses [&NiagaraSpriteRendererProperties]}
  {emitterName=RibbonShortFast, rendererClasses [&NiagaraRibbonRendererProperties]}
  {emitterName=RibbonTallSlow, rendererClasses [&NiagaraRibbonRendererProperties]}
 ]
 userVariables []
}
